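; Actor component for Floor Mist NPCs that shows a friendly indicator marker to the squad of the player who spawned the NPC.
; BeginPlay calls CallOnSpawned, while the visual logic runs in the SpawnInstigator rep-notify on remote clients only.
; The rep-notify compares the instigator pawn's squad ID with the first local player's, and on a match calls ShowIndicatorMarker on the owner's FortPawnComponent_Convert.
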(event EventBeginPlay
  (Default|CallOnSpawned))

(event EventTick
  (params (DeltaSeconds :type Float%20%28single-precision%29)))

(fn OnRep_SpawnInstigator
  (Networking|SwitchHasAuthority
    (:Remote
      (bind _asfort_player_state_athena (Utilities|Casting|CastToFortPlayerStateAthena (Variables|/Script/Engine.Controller|GetPlayerState (Fortnite|GetFirstLocalFortPlayerController true))))
      (bind _asfort_player_pawn_athena (Utilities|Casting|CastToFortPlayerPawnAthena (Variables|Default|GetSpawnInstigator)))
      (bind _asfort_player_state_athena_2 (Utilities|Casting|CastToFortPlayerStateAthena (Variables|/Script/Engine.Pawn|GetPlayerState _asfort_player_pawn_athena)))
      (if (== (Class|FortPlayerStateAthena|GetSquadId _asfort_player_state_athena_2) (Class|FortPlayerStateAthena|GetSquadId _asfort_player_state_athena))
        (Convert|ShowIndicatorMarker (Actor|GetComponentbyClass (Components|GetOwner) "/Script/FortniteGame.FortPawnComponent_Convert"))))))
