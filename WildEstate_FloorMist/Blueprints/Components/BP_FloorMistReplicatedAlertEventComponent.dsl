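; Floor Mist component that raises gameplay alerts for a pawn's owning client, as its event name indicates.
; Its custom event broadcasts the supplied alert data as a stateful event through the global GameplayEventRouter.
; When ShouldNotRepeatForSameSource is set, it checks the AlertSource against the SourcesWhichReceivedAlertFrom set and records it if new.
; BeginPlay and Tick are empty.
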(event EventBeginPlay)

(event EventTick
  (params (DeltaSeconds :type Float%20%28single-precision%29)))

(event Custom|BroadcastGameplayAlertEvent_ReplicatedToPawnOwningClient
  (params (GameplayAlertEventData :type Fort%20Event%20Add%20Gameplay%20Alert%20Requested%20Structure) (ShouldNotRepeatForSameSource :type Boolean) (AlertSource :type Actor%20Object%20Reference))
  (if ShouldNotRepeatForSameSource
    (if (not (Utilities|Set|ContainsItem (Variables|Default|GetSourcesWhichReceivedAlertFrom) AlertSource))
      (Utilities|Set|Add (Variables|Default|GetSourcesWhichReceivedAlertFrom) AlertSource))
    (else
      _))
  (GameplayEventRouter|BroadcastStatefulEvent (GameplayEventRouter|GetGlobalEventRouter) GameplayAlertEventData 0))
